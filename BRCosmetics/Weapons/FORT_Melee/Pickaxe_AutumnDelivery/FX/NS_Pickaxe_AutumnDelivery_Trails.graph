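# Niagara system for the AutumnDelivery pickaxe's swing trails, drawn as ribbons between two bones or sockets on a skeletal mesh.
# BladeTrailMain uses two ribbon renderers and writes a darkened ColorTrailAlt, while BladeTrailHighlight adds a narrower, shorter-lived streak offset further by camera offset.
# Both use FN_*AnimTrail modules and time-of-day tints.
# Alpha fades over particle life, and a second curve indexed by system age quickly cuts both trails off.
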NiagaraSystem /BRCosmetics/Weapons/FORT_Melee/Pickaxe_AutumnDelivery/FX/NS_Pickaxe_AutumnDelivery_Trails
dependencies {
 usedDataInterfaces [&NiagaraDataInterfaceCurve, &NiagaraDataInterfaceColorCurve]
 usedDynamicInputs []
 usedModules [
  &self:SystemSpawnScript.SystemScriptSource.SystemScriptGraph.NiagaraNodeAssignment_0.SetVariables_#1
  &SystemState
 ]
}
emitters {
 BladeTrailHighlight {
  emitterUpdateScript [
   &EmitterState()
   &FN_SpawnAnimTrail(Spawn Interval=0)
  ]
  particleSpawnScript [
   &FN_InitializeAnimTrailBoneIndecies(First Bone or Socket Index=1, Second Bone or Socket Index=0)
   &FN_InitializeAnimTrail(
    AnimTrailWidth=0.45
    Color=&FNTimeOfDayColor(
     Night Color=<rgba 0.17 0.1029 0.0612 0.85>
     Morning Color=<rgba 0.3 0.1461 0.057 0.3>
     Day Color=<rgba 0.44 0.2209 0.0748 0.75>
     Evening Color=<rgba 0.3 0.1461 0.057 0.37>
    )
    Lifetime=0.1
    SkeletalMesh=$System.Skeletal Mesh
   )
   &CameraOffset(Camera Offset Amount=9)
  ]
  particleUpdateScript [
   &FN_UpdateAnimTrail(Locked Particle RibbonWidth=1, SkeletalMesh=$User.SkeletalMesh) {module=FN_UpdateAnimTrail001}
   &ScaleColor(
    Scale Alpha=&FloatFromCurve(FloatCurve=DataInterface() {propertyValues {Curve {keys [
         {arriveTangent=-2.003, interpMode=RCIM_Cubic, leaveTangent=-2.003, tangentMode=RCTM_User, time=0.3, value=1}
         {interpMode=RCIM_Cubic, tangentMode=RCTM_User, time=0.6, value=0}
        ]}}})
    Scale RGB=&MakeVectorFromLinearColorRGB(Color=&ColorFromCurve(ColorCurve=DataInterface() {propertyValues {
        AlphaCurve {keys []}
        BlueCurve {keys [{arriveTangent=-1.3253, interpMode=RCIM_Cubic, time=0.0929, value=1}, <at 0.6199 0.8>]}
        GreenCurve {keys [{arriveTangent=-1.3253, interpMode=RCIM_Cubic, time=0.0929, value=1}, <at 0.6199 0.8>]}
        RedCurve {keys [{arriveTangent=-1.3253, interpMode=RCIM_Cubic, time=0.0929, value=1}, <at 0.6199 0.8>]}
       }}))
    ScaleA=true
    ScaleRGB=false
   )
   &ScaleColor(
    Scale Alpha=&FloatFromCurve(
     FloatCurve=DataInterface() {propertyValues {Curve {keys [{interpMode=RCIM_Cubic, tangentMode=RCTM_User, time=0.1, value=1}, {interpMode=RCIM_Cubic, tangentMode=RCTM_User, time=0.15, value=0}]}}}
     CurveIndex=$System.Age
    )
    ScaleA=true
    ScaleRGB=false
   ) {module=ScaleColor001}
  ]
  rendererClasses [&NiagaraRibbonRendererProperties]
  renderers [{rendererClass=&NiagaraRibbonRendererProperties}]
 }
 BladeTrailMain {
  emitterUpdateScript [
   &EmitterState()
   &FN_SpawnAnimTrail(Spawn Interval=0)
  ]
  particleSpawnScript [
   &FN_InitializeAnimTrailBoneIndecies(First Bone or Socket Index=1, Second Bone or Socket Index=0)
   &FN_InitializeAnimTrail(
    Color=&FNTimeOfDayColor(
     Night Color=<rgba 0.129 0.2042 0.43 1>
     Morning Color=<rgba 0.5 0.39 0.3992 1>
     Day Color=<rgba 0.55 0.495 0.4675 1>
     Evening Color=<rgba 0.46 0.3565 0.322 1>
    )
    Lifetime=0.13
    SkeletalMesh=$System.Skeletal Mesh
   ) {module=FN_InitializeAnimTrail001}
   &CameraOffset(Camera Offset Amount=4)
  ]
  particleUpdateScript [
   &FN_UpdateAnimTrail(Locked Particle RibbonWidth=1, SkeletalMesh=$User.SkeletalMesh) {module=FN_UpdateAnimTrail001}
   &ScaleColor(
    Scale Alpha=&FloatFromCurve(
     FloatCurve=DataInterface() {propertyValues {Curve {keys [
         {arriveTangent=-2.003, interpMode=RCIM_Cubic, leaveTangent=-2.003, tangentMode=RCTM_User, time=0.2, value=1}
         {arriveTangent=-0.2408, interpMode=RCIM_Cubic, leaveTangent=-0.2408, tangentMode=RCTM_User, time=0.7, value=0.05}
         <at 1 0 RCIM_Cubic>
        ]}}}
     Scale Curve=&FNTimeOfDayScalar(Night Scale=0.5, Morning Scale=0.2, Day Scale=0.48, Evening Scale=0.35)
    )
    Scale RGB=&MakeVectorFromLinearColorRGB(Color=&FNTimeOfDayColor(
      Night Color=<rgba 0.129 0.2042 0.43 0.5>
      Morning Color=<rgba 0.5 0.39 0.3992 0.2>
      Day Color=<rgba 0.55 0.4767 0.44 0.48>
      Evening Color=<rgba 0.46 0.3067 0.23 0.35>
     ))
    ScaleA=true
    ScaleRGB=false
   )
   &ScaleColor(
    Scale Alpha=&FloatFromCurve(
     FloatCurve=DataInterface() {propertyValues {Curve {keys [{interpMode=RCIM_Cubic, tangentMode=RCTM_User, time=0.1, value=1}, {interpMode=RCIM_Cubic, tangentMode=RCTM_User, time=0.15, value=0}]}}}
     CurveIndex=$System.Age
    )
    ScaleA=true
    ScaleRGB=false
   ) {module=ScaleColor001}
   /SetVariables(Particles.ColorTrailAlt:LinearColor=&MultiplyLinearColor(Linear Color=$Particles.Color, Scale Factor=<0.75 0.75 0.75 1>)) {isSetParameters=true}
  ]
  rendererClasses [&NiagaraRibbonRendererProperties]
  renderers [
   {rendererClass=&NiagaraRibbonRendererProperties}
   {rendererClass=&NiagaraRibbonRendererProperties, rendererIndex=1}
  ]
 }
}
summary {
 systemName=NS_Pickaxe_AutumnDelivery_Trails
 emitters [
  {emitterName=BladeTrailMain, rendererClasses [&NiagaraRibbonRendererProperties]}
  {emitterName=BladeTrailHighlight, rendererClasses [&NiagaraRibbonRendererProperties]}
 ]
 userVariables [
  User.bisSwinging {type=NiagaraBool}
  User.NewVector3f {type=Vector3f, description="", defaultValue {struct=Vector3f, value=<0 0 0>}}
  User.RibbonIDOffset {type=NiagaraInt32}
  User.SkeletalMesh {
   type=NiagaraDataInterfaceSkeletalMesh
   defaultValue {
    struct=NiagaraExt_VariableValue_DataInterface
    value {dataInterface=&self:NiagaraDataInterfaceSkeletalMesh_0}
   }
  }
  User.Variant {type=NiagaraFloat, defaultValue=0, description=""}
 ]
}
